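# Physics asset for the front-end version of the M_MED_CarrotCake male player body, set up around a dynamic cape.
# Six cape chains (back, front and side, left and right) each link a base and five capsules with tight twist and swing1 limits and wider swing2.
# Fully free constraints join some neighboring chains and the leg bones.
# Thigh, calf and foot capsules cover the legs, and a huge offset floor_joint capsule approximates the ground.
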PhysicsAsset /BRCosmetics/Characters/Player/Male/Medium/Bodies/M_MED_CarrotCake/Meshes/M_MED_CarrotCake_FrontEnd_Physics
bodies {
 calf_l [{shapeType=Capsule, length=29.3669, radius=9.6331, center=<-16.0977 -2 0>, rotation=<rot 90 0 0>}]
 calf_r [{shapeType=Capsule, length=29.3669, radius=9.6331, center=<16.0977 2 0>, rotation=<rot 90 0 0>}]
 dyn_cape_b_l_01 [{
   shapeType=Capsule
   length=21.2466
   radius=6.7586
   shapeName=dyn_cape_b_l_01_capsule
   center=<1.6677 0 3.8876>
   rotation=<rot 90 153.4349 -116.5651>
  }]
 dyn_cape_b_l_02 [{
   shapeType=Capsule
   length=18.837
   radius=9.1682
   shapeName=dyn_cape_b_l_02_capsule
   center=<1.9812 0 7.0019>
   rotation=<rot 90 153.4349 -116.5651>
  }]
 dyn_cape_b_l_03 [{
   shapeType=Capsule
   length=14.6249
   radius=11.0387
   shapeName=dyn_cape_b_l_03_capsule
   center=<1.0001 0 9.084>
   rotation=<rot 90 153.4349 -116.5651>
  }]
 dyn_cape_b_l_04 [{shapeType=Capsule, length=39.808, radius=11.0387, shapeName=dyn_cape_b_l_04_capsule, center=<3.7284 0 8.3782>, rotation=<rot 90 180 -90>}]
 dyn_cape_b_l_05 [{shapeType=Capsule, length=40.0812, radius=11.0387, shapeName=dyn_cape_b_l_05_capsule, center=<4.8609 0 9.0639>, rotation=<rot 90 180 -90>}]
 dyn_cape_b_l_base [{shapeType=Capsule, length=0.505, radius=0.505, shapeName=dyn_cape_b_l_base_capsule, rotation=<rot 0 -90 0>}]
 dyn_cape_b_r_01 [{
   shapeType=Capsule
   length=21.2466
   radius=6.7586
   shapeName=dyn_cape_b_r_01_capsule
   center=<1.6677 0 3.8876>
   rotation=<rot 90 153.4349 -116.5651>
  }]
 dyn_cape_b_r_02 [{
   shapeType=Capsule
   length=18.837
   radius=9.1682
   shapeName=dyn_cape_b_r_02_capsule
   center=<1.9812 0 7.0019>
   rotation=<rot 90 153.4349 -116.5651>
  }]
 dyn_cape_b_r_03 [{
   shapeType=Capsule
   length=14.6249
   radius=11.0387
   shapeName=dyn_cape_b_r_03_capsule
   center=<1.0001 0 9.084>
   rotation=<rot 90 153.4349 -116.5651>
  }]
 dyn_cape_b_r_04 [{
   shapeType=Capsule
   length=39.808
   radius=11.0387
   shapeName=dyn_cape_b_r_04_capsule
   center=<3.7284 0 8.3782>
   rotation=<rot 90 153.4349 -116.5651>
  }]
 dyn_cape_b_r_05 [{shapeType=Capsule, length=40.0812, radius=11.0387, shapeName=dyn_cape_b_r_05_capsule, center=<4.8609 0 9.0639>, rotation=<rot 90 180 -90>}]
 dyn_cape_b_r_base [{shapeType=Capsule, length=0.505, radius=0.505, shapeName=dyn_cape_b_r_base_capsule, rotation=<rot 0 -90 0>}]
 dyn_cape_f_l_01 [{
   shapeType=Capsule
   length=21.2466
   radius=6.7586
   shapeName=dyn_cape_f_l_01_capsule
   center=<1.6677 0 3.8876>
   rotation=<rot 90 153.4349 -116.5651>
  }]
 dyn_cape_f_l_02 [{
   shapeType=Capsule
   length=18.837
   radius=9.1682
   shapeName=dyn_cape_f_l_02_capsule
   center=<1.9812 0 7.0019>
   rotation=<rot 90 153.4349 -116.5651>
  }]
 dyn_cape_f_l_03 [{
   shapeType=Capsule
   length=14.6249
   radius=11.0387
   shapeName=dyn_cape_f_l_03_capsule
   center=<1.0001 0 9.084>
   rotation=<rot 90 153.4349 -116.5651>
  }]
 dyn_cape_f_l_04 [{shapeType=Capsule, length=12.9535, radius=11.0387, shapeName=dyn_cape_f_l_04_capsule, center=<3.7284 0 9.1236>, rotation=<rot 90 180 -90>}]
 dyn_cape_f_l_05 [{shapeType=Capsule, length=16.9665, radius=11.0387, shapeName=dyn_cape_f_l_05_capsule, center=<4.8609 0 9.0639>, rotation=<rot 90 180 -90>}]
 dyn_cape_f_l_base [{shapeType=Capsule, length=0.505, radius=0.505, shapeName=dyn_cape_f_l_base_capsule, rotation=<rot 0 -90 0>}]
 dyn_cape_f_r_01 [{
   shapeType=Capsule
   length=21.2466
   radius=6.7586
   shapeName=dyn_cape_f_r_01_capsule
   center=<1.6677 0 3.8876>
   rotation=<rot 90 153.4349 -116.5651>
  }]
 dyn_cape_f_r_02 [{
   shapeType=Capsule
   length=18.837
   radius=9.1682
   shapeName=dyn_cape_f_r_02_capsule
   center=<1.9812 0 7.0019>
   rotation=<rot 90 153.4349 -116.5651>
  }]
 dyn_cape_f_r_03 [{
   shapeType=Capsule
   length=14.6249
   radius=11.0387
   shapeName=dyn_cape_f_r_03_capsule
   center=<1.0001 0 9.084>
   rotation=<rot 90 153.4349 -116.5651>
  }]
 dyn_cape_f_r_04 [{shapeType=Capsule, length=12.9535, radius=11.0387, shapeName=dyn_cape_f_r_04_capsule, center=<2.8617 0 9.1236>, rotation=<rot 90 180 -90>}]
 dyn_cape_f_r_05 [{shapeType=Capsule, length=12.9535, radius=11.0387, shapeName=dyn_cape_f_r_04_capsule, center=<3.7284 0 9.1236>, rotation=<rot 90 180 -90>}]
 dyn_cape_f_r_base [{shapeType=Capsule, length=0.505, radius=0.505, shapeName=dyn_cape_f_r_base_capsule, rotation=<rot 0 -90 0>}]
 dyn_cape_s_l_01 [{
   shapeType=Capsule
   length=21.2466
   radius=6.7586
   shapeName=dyn_cape_s_l_01_capsule
   center=<1.6677 0 3.8876>
   rotation=<rot 90 153.4349 -116.5651>
  }]
 dyn_cape_s_l_02 [{
   shapeType=Capsule
   length=18.837
   radius=9.1682
   shapeName=dyn_cape_s_l_02_capsule
   center=<1.9812 0 7.0019>
   rotation=<rot 90 153.4349 -116.5651>
  }]
 dyn_cape_s_l_03 [{
   shapeType=Capsule
   length=14.6249
   radius=11.0387
   shapeName=dyn_cape_s_l_03_capsule
   center=<1.0001 0 9.084>
   rotation=<rot 90 153.4349 -116.5651>
  }]
 dyn_cape_s_l_04 [{shapeType=Capsule, length=12.9535, radius=11.0387, shapeName=dyn_cape_s_l_04_capsule, center=<3.7284 0 9.1236>, rotation=<rot 90 180 -90>}]
 dyn_cape_s_l_05 [{shapeType=Capsule, length=16.9665, radius=11.0387, shapeName=dyn_cape_s_l_05_capsule, center=<4.8609 0 9.0639>, rotation=<rot 90 180 -90>}]
 dyn_cape_s_l_base [{shapeType=Capsule, length=0.505, radius=0.505, shapeName=dyn_cape_s_l_base_capsule, rotation=<rot 0 -90 0>}]
 dyn_cape_s_r_01 [{
   shapeType=Capsule
   length=21.2466
   radius=6.7586
   shapeName=dyn_cape_s_r_01_capsule
   center=<1.6677 0 3.8876>
   rotation=<rot 90 153.4349 -116.5651>
  }]
 dyn_cape_s_r_02 [{
   shapeType=Capsule
   length=18.837
   radius=9.1682
   shapeName=dyn_cape_s_r_02_capsule
   center=<1.9812 0 7.0019>
   rotation=<rot 90 153.4349 -116.5651>
  }]
 dyn_cape_s_r_03 [{
   shapeType=Capsule
   length=14.6249
   radius=11.0387
   shapeName=dyn_cape_s_r_03_capsule
   center=<1.0001 0 9.084>
   rotation=<rot 90 153.4349 -116.5651>
  }]
 dyn_cape_s_r_04 [{shapeType=Capsule, length=25.4042, radius=11.0387, shapeName=dyn_cape_s_r_04_capsule, center=<3.7284 0 9.1236>, rotation=<rot 90 180 -90>}]
 dyn_cape_s_r_05 [{
   shapeType=Capsule
   length=32.0605
   radius=11.0387
   shapeName=dyn_cape_s_r_05_capsule
   center=<4.8609 1.4522 9.0639>
   rotation=<rot 90 180 -90>
  }]
 dyn_cape_s_r_base [{shapeType=Capsule, length=0.505, radius=0.505, shapeName=dyn_cape_s_r_base_capsule, rotation=<rot 0 -90 0>}]
 floor_joint [{
   shapeType=Capsule
   length=0.5056
   radius=337.6725
   shapeName=floor_joint_capsule
   center=<2.6954 -23.3737 -324.4869>
   rotation=<rot 0 90 -90>
  }]
 foot_l [{shapeType=Capsule, length=15, radius=12.75, shapeName=foot_l_capsule, center=<-4.0948 6.1521 -0.1893>, rotation=<rot 0 0 90>}]
 foot_r [{shapeType=Capsule, length=15, radius=12.25, shapeName=foot_r_capsule, center=<4.0948 -4.8644 0.1893>, rotation=<rot 0 0 -90>}]
 thigh_l [{shapeType=Capsule, length=29.3943, radius=10.8557, shapeName=thigh_l_capsule, center=<-15.7821 -2.0135 0.1765>, rotation=<rot 90 0 0>}]
 thigh_r [{shapeType=Capsule, length=29.3943, radius=10.8557, shapeName=thigh_r_capsule, center=<15.4135 1.0948 0.028>, rotation=<rot 90 0 0>}]
}
constraints [
 {
  bone1Name=calf_l
  bone2Name=thigh_l
  swing1LimitDegrees=45
  swing1Motion=Free
  swing2LimitDegrees=45
  swing2Motion=Free
  twistLimitDegrees=45
  twistMotion=Free
 }
 {
  bone1Name=foot_l
  bone2Name=calf_l
  swing1LimitDegrees=45
  swing1Motion=Free
  swing2LimitDegrees=45
  swing2Motion=Free
  twistLimitDegrees=45
  twistMotion=Free
 }
 {
  bone1Name=calf_r
  bone2Name=thigh_r
  swing1LimitDegrees=45
  swing1Motion=Free
  swing2LimitDegrees=45
  swing2Motion=Free
  twistLimitDegrees=45
  twistMotion=Free
 }
 {
  bone1Name=foot_r
  bone2Name=calf_r
  swing1LimitDegrees=45
  swing1Motion=Free
  swing2LimitDegrees=45
  swing2Motion=Free
  twistLimitDegrees=45
  twistMotion=Free
 }
 {bone1Name=dyn_cape_b_l_01, bone2Name=dyn_cape_b_l_base, swing1LimitDegrees=2, swing2LimitDegrees=30, twistLimitDegrees=1}
 {bone1Name=dyn_cape_b_l_02, bone2Name=dyn_cape_b_l_01, swing1LimitDegrees=2, swing2LimitDegrees=25, twistLimitDegrees=1}
 {bone1Name=dyn_cape_b_l_03, bone2Name=dyn_cape_b_l_02, swing1LimitDegrees=2, swing2LimitDegrees=20, twistLimitDegrees=1}
 {bone1Name=dyn_cape_b_l_04, bone2Name=dyn_cape_b_l_03, swing1LimitDegrees=2, swing2LimitDegrees=15, twistLimitDegrees=1}
 {bone1Name=dyn_cape_b_l_05, bone2Name=dyn_cape_b_l_04, swing1LimitDegrees=2, swing2LimitDegrees=10, twistLimitDegrees=1}
 {bone1Name=dyn_cape_b_r_01, bone2Name=dyn_cape_b_r_base, swing1LimitDegrees=2, swing2LimitDegrees=30, twistLimitDegrees=1}
 {bone1Name=dyn_cape_b_r_02, bone2Name=dyn_cape_b_r_01, swing1LimitDegrees=2, swing2LimitDegrees=25, twistLimitDegrees=1}
 {bone1Name=dyn_cape_b_r_03, bone2Name=dyn_cape_b_r_02, swing1LimitDegrees=2, swing2LimitDegrees=20, twistLimitDegrees=1}
 {bone1Name=dyn_cape_b_r_04, bone2Name=dyn_cape_b_r_03, swing1LimitDegrees=2, swing2LimitDegrees=15, twistLimitDegrees=1}
 {bone1Name=dyn_cape_b_r_05, bone2Name=dyn_cape_b_r_04, swing1LimitDegrees=2, swing2LimitDegrees=10, twistLimitDegrees=1}
 {bone1Name=dyn_cape_f_l_01, bone2Name=dyn_cape_f_l_base, swing1LimitDegrees=2, swing2LimitDegrees=7.5, twistLimitDegrees=1}
 {bone1Name=dyn_cape_f_l_02, bone2Name=dyn_cape_f_l_01, swing1LimitDegrees=2, swing2LimitDegrees=15, twistLimitDegrees=1}
 {bone1Name=dyn_cape_f_l_03, bone2Name=dyn_cape_f_l_02, swing1LimitDegrees=2, swing2LimitDegrees=20, twistLimitDegrees=1}
 {bone1Name=dyn_cape_f_l_04, bone2Name=dyn_cape_f_l_03, swing1LimitDegrees=2, swing2LimitDegrees=15, twistLimitDegrees=1}
 {bone1Name=dyn_cape_f_l_05, bone2Name=dyn_cape_f_l_04, swing1LimitDegrees=2, swing2LimitDegrees=10, twistLimitDegrees=1}
 {bone1Name=dyn_cape_f_r_01, bone2Name=dyn_cape_f_r_base, swing1LimitDegrees=2, swing2LimitDegrees=15, twistLimitDegrees=1}
 {bone1Name=dyn_cape_f_r_02, bone2Name=dyn_cape_f_r_01, swing1LimitDegrees=2, swing2LimitDegrees=25, twistLimitDegrees=1}
 {bone1Name=dyn_cape_f_r_03, bone2Name=dyn_cape_f_r_02, swing1LimitDegrees=2, swing2LimitDegrees=20, twistLimitDegrees=1}
 {bone1Name=dyn_cape_f_r_04, bone2Name=dyn_cape_f_r_03, swing1LimitDegrees=2, swing2LimitDegrees=15, twistLimitDegrees=1}
 {bone1Name=dyn_cape_f_r_05, bone2Name=dyn_cape_f_r_04, swing1LimitDegrees=2, swing2LimitDegrees=10, twistLimitDegrees=1}
 {bone1Name=dyn_cape_s_l_01, bone2Name=dyn_cape_s_l_base, swing1LimitDegrees=2, swing2LimitDegrees=7.5, twistLimitDegrees=1}
 {bone1Name=dyn_cape_s_l_02, bone2Name=dyn_cape_s_l_01, swing1LimitDegrees=2, swing2LimitDegrees=15, twistLimitDegrees=1}
 {bone1Name=dyn_cape_s_l_03, bone2Name=dyn_cape_s_l_02, swing1LimitDegrees=2, swing2LimitDegrees=20, twistLimitDegrees=1}
 {bone1Name=dyn_cape_s_l_04, bone2Name=dyn_cape_s_l_03, swing1LimitDegrees=2, swing2LimitDegrees=15, twistLimitDegrees=1}
 {bone1Name=dyn_cape_s_l_05, bone2Name=dyn_cape_s_l_04, swing1LimitDegrees=2, swing2LimitDegrees=10, twistLimitDegrees=1}
 {bone1Name=dyn_cape_s_r_01, bone2Name=dyn_cape_s_r_base, swing1LimitDegrees=2, swing2LimitDegrees=30, twistLimitDegrees=1}
 {bone1Name=dyn_cape_s_r_02, bone2Name=dyn_cape_s_r_01, swing1LimitDegrees=2, swing2LimitDegrees=25, twistLimitDegrees=1}
 {bone1Name=dyn_cape_s_r_03, bone2Name=dyn_cape_s_r_02, swing1LimitDegrees=2, swing2LimitDegrees=20, twistLimitDegrees=1}
 {bone1Name=dyn_cape_s_r_04, bone2Name=dyn_cape_s_r_03, swing1LimitDegrees=2, swing2LimitDegrees=15, twistLimitDegrees=1}
 {bone1Name=dyn_cape_s_r_05, bone2Name=dyn_cape_s_r_04, swing1LimitDegrees=2, swing2LimitDegrees=10, twistLimitDegrees=1}
 {
  bone1Name=dyn_cape_b_l_04
  bone2Name=dyn_cape_s_l_04
  swing1LimitDegrees=45
  swing1Motion=Free
  swing2LimitDegrees=45
  swing2Motion=Free
  twistLimitDegrees=45
  twistMotion=Free
 }
 {
  bone1Name=dyn_cape_b_l_03
  bone2Name=dyn_cape_b_r_03
  swing1LimitDegrees=45
  swing1Motion=Free
  swing2LimitDegrees=45
  swing2Motion=Free
  twistLimitDegrees=45
  twistMotion=Free
 }
 {
  bone1Name=dyn_cape_s_l_04
  bone2Name=dyn_cape_f_l_04
  swing1LimitDegrees=45
  swing1Motion=Free
  swing2LimitDegrees=45
  swing2Motion=Free
  twistLimitDegrees=45
  twistMotion=Free
 }
 {
  bone1Name=dyn_cape_s_r_04
  bone2Name=dyn_cape_f_r_04
  swing1LimitDegrees=45
  swing1Motion=Free
  swing2LimitDegrees=45
  swing2Motion=Free
  twistLimitDegrees=45
  twistMotion=Free
 }
 {
  bone1Name=dyn_cape_s_r_03
  bone2Name=dyn_cape_f_r_03
  swing1LimitDegrees=45
  swing1Motion=Free
  swing2LimitDegrees=45
  swing2Motion=Free
  twistLimitDegrees=45
  twistMotion=Free
 }
]
