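# Physics asset for the neck mesh of the male medium CarrotCake cosmetic body, covering six coat segments around the neck for secondary motion.
# Each segment pairs a small base_neckCoat capsule with a larger, offset dyn_neckCoat capsule, joined by tight swing and very narrow twist limits.
# All six base bodies attach to a large spine_04 torso capsule through constraints with free swing and twist.
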PhysicsAsset /BRCosmetics/Characters/Player/Male/Medium/Bodies/M_MED_CarrotCake/Meshes/M_MED_CarrotCake_Neck_Physics
bodies {
 base_neckCoat_01 [{shapeType=Capsule, length=0.505, radius=0.505, shapeName=base_neckCoat_01_capsule, rotation=<rot 0 -90 0>}]
 base_neckCoat_02 [{shapeType=Capsule, length=0.505, radius=0.505, shapeName=base_neckCoat_02_capsule, rotation=<rot 0 -90 0>}]
 base_neckCoat_03 [{shapeType=Capsule, length=0.505, radius=0.505, shapeName=base_neckCoat_03_capsule, rotation=<rot 0 -90 0>}]
 base_neckCoat_04 [{shapeType=Capsule, length=0.505, radius=0.505, shapeName=base_neckCoat_04_capsule, rotation=<rot 0 -90 0>}]
 base_neckCoat_05 [{shapeType=Capsule, length=0.505, radius=0.505, shapeName=base_neckCoat_05_capsule, rotation=<rot 0 -90 0>}]
 base_neckCoat_06 [{shapeType=Capsule, length=0.505, radius=0.505, shapeName=base_neckCoat_06_capsule, rotation=<rot 0 -90 0>}]
 dyn_neckCoat_01 [{shapeType=Capsule, length=2, radius=1, shapeName=dyn_neckCoat_01_capsule, center=<2 0 0>, rotation=<rot 90 0 0>}]
 dyn_neckCoat_02 [{shapeType=Capsule, length=2, radius=1, shapeName=dyn_neckCoat_02_capsule, center=<2 0 0>, rotation=<rot 90 0 0>}]
 dyn_neckCoat_03 [{shapeType=Capsule, length=2, radius=1, shapeName=dyn_neckCoat_03_capsule, center=<2 0 0>, rotation=<rot 90 0 0>}]
 dyn_neckCoat_04 [{shapeType=Capsule, length=2, radius=1, shapeName=dyn_neckCoat_04_capsule, center=<2 0 0>, rotation=<rot 90 0 0>}]
 dyn_neckCoat_05 [{shapeType=Capsule, length=2, radius=1, shapeName=dyn_neckCoat_05_capsule, center=<2 0 0>, rotation=<rot 90 0 0>}]
 dyn_neckCoat_06 [{shapeType=Capsule, length=2, radius=1, shapeName=dyn_neckCoat_06_capsule, center=<2 0 0>, rotation=<rot 90 0 0>}]
 spine_04 [{shapeType=Capsule, length=13.6074, radius=10.9956, shapeName=spine_04_capsule}]
}
constraints [
 {bone1Name=dyn_neckCoat_01, bone2Name=base_neckCoat_01, swing1LimitDegrees=8, swing2LimitDegrees=15, twistLimitDegrees=1}
 {bone1Name=dyn_neckCoat_02, bone2Name=base_neckCoat_02, swing1LimitDegrees=8, swing2LimitDegrees=15, twistLimitDegrees=1}
 {bone1Name=dyn_neckCoat_03, bone2Name=base_neckCoat_03, swing1LimitDegrees=8, swing2LimitDegrees=15, twistLimitDegrees=1}
 {bone1Name=dyn_neckCoat_04, bone2Name=base_neckCoat_04, swing1LimitDegrees=8, swing2LimitDegrees=15, twistLimitDegrees=1}
 {bone1Name=dyn_neckCoat_05, bone2Name=base_neckCoat_05, swing1LimitDegrees=8, swing2LimitDegrees=15, twistLimitDegrees=1}
 {bone1Name=dyn_neckCoat_06, bone2Name=base_neckCoat_06, swing1LimitDegrees=8, swing2LimitDegrees=15, twistLimitDegrees=1}
 {
  bone1Name=base_neckCoat_01
  bone2Name=spine_04
  swing1LimitDegrees=45
  swing1Motion=Free
  swing2LimitDegrees=45
  swing2Motion=Free
  twistLimitDegrees=45
  twistMotion=Free
 }
 {
  bone1Name=base_neckCoat_02
  bone2Name=spine_04
  swing1LimitDegrees=45
  swing1Motion=Free
  swing2LimitDegrees=45
  swing2Motion=Free
  twistLimitDegrees=45
  twistMotion=Free
 }
 {
  bone1Name=base_neckCoat_03
  bone2Name=spine_04
  swing1LimitDegrees=45
  swing1Motion=Free
  swing2LimitDegrees=45
  swing2Motion=Free
  twistLimitDegrees=45
  twistMotion=Free
 }
 {
  bone1Name=base_neckCoat_04
  bone2Name=spine_04
  swing1LimitDegrees=45
  swing1Motion=Free
  swing2LimitDegrees=45
  swing2Motion=Free
  twistLimitDegrees=45
  twistMotion=Free
 }
 {
  bone1Name=base_neckCoat_05
  bone2Name=spine_04
  swing1LimitDegrees=45
  swing1Motion=Free
  swing2LimitDegrees=45
  swing2Motion=Free
  twistLimitDegrees=45
  twistMotion=Free
 }
 {
  bone1Name=base_neckCoat_06
  bone2Name=spine_04
  swing1LimitDegrees=45
  swing1Motion=Free
  swing2LimitDegrees=45
  swing2Motion=Free
  twistLimitDegrees=45
  twistMotion=Free
 }
]
