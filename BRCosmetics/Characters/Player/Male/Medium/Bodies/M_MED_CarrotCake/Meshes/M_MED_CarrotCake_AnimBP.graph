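# Cosmetic AnimBP for the CarrotCake male player body, adding coat/cape motion and correction to the pose copied from the parent mesh.
# A Control Rig pass feeds one of two additive bone-offset chains, picked by bRigidBodyEnabled and cached as Pre-dynamics.
# RigidBody variants for front end, skydive cape or default follow, then a neck simulation.
# The EventGraph converts movement states into float alphas and updates gravity override.
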AnimBlueprint /BRCosmetics/Characters/Player/Male/Medium/Bodies/M_MED_CarrotCake/Meshes/M_MED_CarrotCake_AnimBP
target_skeleton=&/M_MED_CarrotCake_Skeleton
graphs [
 AnimGraph {nodes [
   n1 = Get Is Sliding() -> (Is Sliding:double)
   n2 = "Transform (Modify) Bone - Bone: dyn_cape_s_l_base"(Alpha=@n1, Translation="0,0,0", Rotation="0,0,0", Scale="1,1,1") {
    rotationMode=BMM_Additive
    rotationSpace=BCS_BoneSpace
    translationMode=BMM_Additive
    translationSpace=BCS_BoneSpace
    alphaBoolBlend {blendOption=Linear}
    alphaScaleBiasClamp {bInterpResult=true}
    boneToModify {boneName=dyn_cape_s_l_base}
   }
   "Transform (Modify) Bone - Bone: dyn_cape_f_l_base"(ComponentPose=@n2, Alpha=@n1, Translation="0,0,0", Rotation="0,0,0", Scale="1,1,1") {
    rotationMode=BMM_Additive
    rotationSpace=BCS_BoneSpace
    translationMode=BMM_Additive
    translationSpace=BCS_BoneSpace
    alphaBoolBlend {blendOption=Linear}
    alphaScaleBiasClamp {bInterpResult=true}
    boneToModify {boneName=dyn_cape_f_l_base}
   }
   "Transform (Modify) Bone - Bone: dyn_cape_f_r_base"(Alpha=@n1, Translation="0,0,0", Rotation="0,0,0", Scale="1,1,1") {
    rotationMode=BMM_Additive
    rotationSpace=BCS_BoneSpace
    translationMode=BMM_Additive
    translationSpace=BCS_BoneSpace
    alphaBoolBlend {blendOption=Linear}
    alphaScaleBiasClamp {bInterpResult=true}
    boneToModify {boneName=dyn_cape_f_r_base}
   }
   n3 = Get Inside Car() -> (Inside Car:double)
   n4 = "Transform (Modify) Bone - Bone: dyn_cape_f_l_base"(Alpha=@n3, Translation="-50,0,0", Rotation="0,0,0", Scale="1,1,1") {
    rotationMode=BMM_Additive
    rotationSpace=BCS_BoneSpace
    translationMode=BMM_Additive
    translationSpace=BCS_BoneSpace
    alphaBoolBlend {blendOption=Linear}
    alphaScaleBiasClamp {bInterpResult=true}
    boneToModify {boneName=dyn_cape_f_l_base}
    translation {x=-50}
   }
   n5 = "Transform (Modify) Bone - Bone: dyn_cape_f_r_base"(ComponentPose=@n4, Alpha=@n3, Translation="-50,0,0", Rotation="0,0,0", Scale="1,1,1") {
    rotationMode=BMM_Additive
    rotationSpace=BCS_BoneSpace
    translationMode=BMM_Additive
    translationSpace=BCS_BoneSpace
    alphaBoolBlend {blendOption=Linear}
    alphaScaleBiasClamp {bInterpResult=true}
    boneToModify {boneName=dyn_cape_f_r_base}
    translation {x=-50}
   }
   n6 = "Transform (Modify) Bone - Bone: dyn_cape_s_r_base"(ComponentPose=@n5, Alpha=@n3, Translation="-50,0,0", Rotation="0,0,0", Scale="1,1,1") {
    rotationMode=BMM_Additive
    rotationSpace=BCS_BoneSpace
    translationMode=BMM_Additive
    translationSpace=BCS_BoneSpace
    alphaBoolBlend {blendOption=Linear}
    alphaScaleBiasClamp {bInterpResult=true}
    boneToModify {boneName=dyn_cape_s_r_base}
    translation {x=-50}
   }
   n7 = "Transform (Modify) Bone - Bone: dyn_cape_s_l_base"(ComponentPose=@n6, Alpha=@n3, Translation="-50,0,0", Rotation="0,0,0", Scale="1,1,1") {
    rotationMode=BMM_Additive
    rotationSpace=BCS_BoneSpace
    translationMode=BMM_Additive
    translationSpace=BCS_BoneSpace
    alphaBoolBlend {blendOption=Linear}
    alphaScaleBiasClamp {bInterpResult=true}
    boneToModify {boneName=dyn_cape_s_l_base}
    translation {x=-50}
   }
   n8 = "Transform (Modify) Bone - Bone: dyn_cape_b_l_base"(ComponentPose=@n7, Alpha=@n3, Translation="-50,0,10", Rotation="0,0,0", Scale="1,1,1") {
    rotationMode=BMM_Additive
    rotationSpace=BCS_BoneSpace
    translationMode=BMM_Additive
    translationSpace=BCS_BoneSpace
    alphaBoolBlend {blendOption=Linear}
    alphaScaleBiasClamp {bInterpResult=true}
    boneToModify {boneName=dyn_cape_b_l_base}
    translation {x=-50, z=10}
   }
   "Transform (Modify) Bone - Bone: dyn_cape_b_r_base"(ComponentPose=@n8, Alpha=@n3, Translation="-50,0,10", Rotation="0,0,0", Scale="1,1,1") {
    rotationMode=BMM_Additive
    rotationSpace=BCS_BoneSpace
    translationMode=BMM_Additive
    translationSpace=BCS_BoneSpace
    alphaBoolBlend {blendOption=Linear}
    alphaScaleBiasClamp {bInterpResult=true}
    boneToModify {boneName=dyn_cape_b_r_base}
    translation {x=-50, z=10}
   }
   RigidBody!(
    ComponentPose
    Alpha
    OverrideWorldGravity:Vector=Get Gravity Override()
    ExternalForce
    SimSpaceSettings:Sim Space Settings Structure=(WorldAlpha=0,VelocityScaleZ=1,DampingAlpha=1,MaxLinearVelocity=10000,MaxAngularVelocity=10000,MaxLinearAcceleration=10000,MaxAngularAcceleration=10000,ExternalLinearDragV=(X=0,Y=0,Z=0),ExternalLinearVelocity=(X=0,Y=0,Z=0),ExternalAngularVelocity=(X=0,Y=0,Z=0))
   ) -> (Pose) {
    bClampLinearTranslationLimitToRefPose=true
    bOverrideWorldGravity=true
    overridePhysicsAsset=&/M_MED_CarrotCake_Physics
    simulationSpace=BaseBoneSpace
    alphaBoolBlend {blendOption=Linear}
    baseBoneRef {boneName=pelvis}
    componentAppliedLinearAccClamp=<3000 3000 1000>
    componentLinearAccScale=<0.5 0.4 0.4>
    componentLinearVelScale=<0.5 0.4 0.4>
    overrideWorldGravity {z=-920}
   }
   n9 = RigidBody!(
    ComponentPose=Local To Component(LocalPose=Use cached pose 'Pre-dynamics'())
    Alpha
    OverrideWorldGravity="0,0,-1000"
    ExternalForce="0,0,-980"
   ) -> (Pose) {
    bClampLinearTranslationLimitToRefPose=true
    bOverrideWorldGravity=true
    overridePhysicsAsset=&M_MED_CarrotCake_Physics
    simulationSpace=BaseBoneSpace
    alphaBoolBlend {blendOption=Linear}
    baseBoneRef {boneName=root}
    componentAppliedLinearAccClamp=<3000 3000 1000>
    componentLinearAccScale=<0.1 0.1 0.2>
    componentLinearVelScale=<0.1 0.1 0.2>
    overrideWorldGravity {z=-1000}
    simSpaceSettings {worldAlpha=0.1}
   }
   n10 = RigidBody(ComponentPose=Local To Component(LocalPose=Use cached pose 'Pre-dynamics'()), OverrideWorldGravity="0,0,-920") {
    bClampLinearTranslationLimitToRefPose=true
    bOverrideWorldGravity=true
    overridePhysicsAsset=&/M_MED_CarrotCake_Skydive_Physics
    simulationSpace=BaseBoneSpace
    alphaBoolBlend {blendOption=Linear}
    baseBoneRef {boneName=root}
    componentAppliedLinearAccClamp=<1000 1000 1000>
    componentLinearAccScale=<0.01 0.01 0.01>
    componentLinearVelScale=<0.01 0.01 0.01>
    overrideWorldGravity {z=-920}
    simSpaceSettings {worldAlpha=0.1}
   }
   n11 = Blend Poses by bool(
    BlendPose_0=Component To Local(ComponentPose=@n10)
    BlendPose_1=Component To Local(ComponentPose=@n9)
    bActiveValue=Get Is Skydive Cape Equiped() -> (Is Skydive Cape Equiped:bool)
   ) {blendType=Linear}
   n12 = RigidBody(ComponentPose=Local To Component(LocalPose=Use cached pose 'Pre-dynamics'()), OverrideWorldGravity="0,0,-920") {
    bClampLinearTranslationLimitToRefPose=true
    bOverrideWorldGravity=true
    overridePhysicsAsset=&/M_MED_CarrotCake_FrontEnd_Physics
    simulationSpace=BaseBoneSpace
    alphaBoolBlend {blendOption=Linear}
    baseBoneRef {boneName=root}
    componentAppliedLinearAccClamp=<1000 1000 1000>
    componentLinearAccScale=<0.01 0.01 0.01>
    componentLinearVelScale=<0.01 0.01 0.01>
    overrideWorldGravity {z=-920}
    simSpaceSettings {velocityScaleZ=0.5, worldAlpha=0.1}
   }
   n13 = Component To Local(ComponentPose=@n12)
   n14 = Get bIsInFrontEnd()
   n15 = RigidBody!(
    ComponentPose=Local To Component(LocalPose=Blend Poses by bool(BlendPose_0=@n13, BlendPose_1=@n11, bActiveValue=@n14) {blendType=Linear})
    Alpha
    OverrideWorldGravity=Get Gravity Override()
   ) -> (Pose) {
    bClampLinearTranslationLimitToRefPose=true
    bOverrideWorldGravity=true
    overridePhysicsAsset=&/M_MED_CarrotCake_Neck_Physics
    alphaBoolBlend {blendOption=Linear}
    baseBoneRef {boneName=spine_03}
    componentAppliedLinearAccClamp=<1000 1000 1000>
    componentLinearAccScale=<0.3 0.3 0.3>
    componentLinearVelScale=<0.3 0.3 0.3>
    simSpaceSettings {
     maxAngularAcceleration=1000
     maxAngularVelocity=3
     maxLinearAcceleration=700
     maxLinearVelocity=500
     velocityScaleZ=0
     worldAlpha=0.5
     externalLinearDragV=<0.2 0.2>
    }
   }
   Output Pose(Result=Component To Local(ComponentPose=@n15))
   n16 = Control Rig!(
    Source=Copy Pose From Mesh() {bUseAttachedParent=true}
    Alpha
    Standing Fixes Coat:double=Get Standing Fixes Coat() -> (Standing Fixes Coat:double)
    isMoving2D:double=0
    isGlidingCtrl:bool=false
   ) -> (Pose) {
    alphaBoolBlend {blendOption=Linear}
    controlRigAssetReference {blueprintRigClass=&/M_MED_CarrotCake_CtrlRig.M_MED_CarrotCake_CtrlRig_C}
   }
   Save cached pose 'ControlRig '(Pose=@n16) {cacheName="ControlRig "}
   n17 = Get Is Parachuting() -> (Is Parachuting:double)
   n18 = Get Moving 2d() -> (Moving 2d:double)
   n19 = "Transform (Modify) Bone - Bone: dyn_cape_b_l_base"(
    ComponentPose=Local To Component(LocalPose=Use cached pose 'ControlRig '())
    Alpha=@n18
    Translation="0,0,-5"
    Rotation="35,0,0"
    Scale="1,1,1"
   ) {
    rotationMode=BMM_Additive
    rotationSpace=BCS_BoneSpace
    translationMode=BMM_Additive
    translationSpace=BCS_BoneSpace
    alphaBoolBlend {blendOption=Linear}
    boneToModify {boneName=dyn_cape_b_l_base}
    rotation {pitch=35}
    translation {z=-5}
   }
   n20 = "Transform (Modify) Bone - Bone: dyn_cape_b_r_base"(ComponentPose=@n19, Alpha=@n18, Translation="0,0,-5", Rotation="35,0,0", Scale="1,1,1") {
    rotationMode=BMM_Additive
    rotationSpace=BCS_BoneSpace
    translationMode=BMM_Additive
    translationSpace=BCS_BoneSpace
    alphaBoolBlend {blendOption=Linear}
    boneToModify {boneName=dyn_cape_b_r_base}
    rotation {pitch=35}
    translation {z=-5}
   }
   n21 = "Transform (Modify) Bone - Bone: back_coat_helper_02"(ComponentPose=@n20, Alpha=Get Moving2D BackHelper() -> (Moving2D BackHelper:double), Translation="-15,0,0", Rotation="0,0,0", Scale="1,1,1") {
    rotationMode=BMM_Additive
    rotationSpace=BCS_BoneSpace
    translationMode=BMM_Additive
    translationSpace=BCS_BoneSpace
    alphaBoolBlend {blendOption=Linear}
    alphaScaleBiasClamp {bInterpResult=true}
    boneToModify {boneName=back_coat_helper_02}
    translation {x=-15}
   }
   n22 = "Transform (Modify) Bone - Bone: dyn_cape_b_l_base"(ComponentPose=@n21, Alpha=@n1, Translation="0,0,0", Rotation="35,0,0", Scale="1,1,1") {
    rotationMode=BMM_Additive
    rotationSpace=BCS_BoneSpace
    translationMode=BMM_Additive
    translationSpace=BCS_BoneSpace
    alphaBoolBlend {blendOption=Linear}
    alphaScaleBiasClamp {bInterpResult=true}
    boneToModify {boneName=dyn_cape_b_l_base}
    rotation {pitch=35}
   }
   n23 = "Transform (Modify) Bone - Bone: dyn_cape_s_r_base"(ComponentPose=@n22, Alpha=@n1, Translation="0,0,0", Rotation="60,0,0", Scale="1,1,1") {
    rotationMode=BMM_Additive
    rotationSpace=BCS_BoneSpace
    translationMode=BMM_Additive
    translationSpace=BCS_BoneSpace
    alphaBoolBlend {blendOption=Linear}
    alphaScaleBiasClamp {bInterpResult=true}
    boneToModify {boneName=dyn_cape_s_r_base}
    rotation {pitch=60}
   }
   n24 = "Transform (Modify) Bone - Bone: dyn_cape_b_r_base"(ComponentPose=@n23, Alpha=@n1, Translation="0,0,0", Rotation="90,0,0", Scale="1,1,1") {
    rotationMode=BMM_Additive
    rotationSpace=BCS_BoneSpace
    translationMode=BMM_Additive
    translationSpace=BCS_BoneSpace
    alphaBoolBlend {blendOption=Linear}
    alphaScaleBiasClamp {bInterpResult=true}
    boneToModify {boneName=dyn_cape_b_r_base}
    rotation {pitch=90}
   }
   n25 = "Transform (Modify) Bone - Bone: back_coat_helper_03"(ComponentPose=@n24, Alpha=Get Is Skydiving() -> (Is Skydiving:double), Translation="0,0,0", Rotation="0,0,0", Scale="1,0.85,1") {
    rotationMode=BMM_Additive
    rotationSpace=BCS_BoneSpace
    scaleMode=BMM_Additive
    scaleSpace=BCS_BoneSpace
    translationMode=BMM_Additive
    translationSpace=BCS_BoneSpace
    alphaBoolBlend {blendOption=Linear}
    alphaScaleBiasClamp {bInterpResult=true}
    boneToModify {boneName=back_coat_helper_03}
    scale {y=0.85}
   }
   n26 = Get is FrontEnd() -> (is FrontEnd:double)
   n27 = "Transform (Modify) Bone - Bone: dyn_cape_f_l_base"(ComponentPose=@n25, Alpha=@n26, Translation="0.829,-1.412,-0.670", Rotation="0,0,0", Scale="1,1,1") {
    rotationMode=BMM_Additive
    rotationSpace=BCS_BoneSpace
    translationMode=BMM_Additive
    translationSpace=BCS_BoneSpace
    alphaBoolBlend {blendOption=Linear}
    boneToModify {boneName=dyn_cape_f_l_base}
    translation=<0.829 -1.412 -0.67>
   }
   n28 = "Transform (Modify) Bone - Bone: dyn_cape_s_l_base"(ComponentPose=@n27, Alpha=@n26, Translation="-0.759,-1.693,-3.121", Rotation="0,0,0", Scale="1,1,1") {
    rotationMode=BMM_Additive
    rotationSpace=BCS_BoneSpace
    translationMode=BMM_Additive
    translationSpace=BCS_BoneSpace
    alphaBoolBlend {blendOption=Linear}
    boneToModify {boneName=dyn_cape_s_l_base}
    translation=<-0.759 -1.693 -3.121>
   }
   n29 = "Transform (Modify) Bone - Bone: dyn_cape_b_l_base"(ComponentPose=@n28, Alpha=@n17, Translation="0,0,0", Rotation="0,0,0", Scale="1,1,1") {
    rotationMode=BMM_Additive
    rotationSpace=BCS_BoneSpace
    translationMode=BMM_Additive
    translationSpace=BCS_BoneSpace
    alphaBoolBlend {blendOption=Linear}
    boneToModify {boneName=dyn_cape_b_l_base}
   }
   n30 = "Transform (Modify) Bone - Bone: dyn_cape_s_r_base"(ComponentPose=@n29, Alpha=@n17, Translation="0,0,0", Rotation="0,0,0", Scale="1,1,1") {
    rotationMode=BMM_Additive
    rotationSpace=BCS_BoneSpace
    translationMode=BMM_Additive
    translationSpace=BCS_BoneSpace
    alphaBoolBlend {blendOption=Linear}
    boneToModify {boneName=dyn_cape_s_r_base}
   }
   n31 = Get CrouchingMoving() -> (CrouchingMoving:double)
   n32 = "Transform (Modify) Bone - Bone: back_coat_helper_01"(ComponentPose=@n30, Alpha=@n31, Translation="-7,0,0", Rotation="0,0,0", Scale="1,1,1") {
    rotationMode=BMM_Additive
    rotationSpace=BCS_BoneSpace
    translationMode=BMM_Additive
    translationSpace=BCS_BoneSpace
    alphaBoolBlend {blendOption=Linear}
    alphaScaleBiasClamp {bInterpResult=true}
    boneToModify {boneName=back_coat_helper_01}
    translation {x=-7}
   }
   n33 = "Transform (Modify) Bone - Bone: back_coat_helper_02"(ComponentPose=@n32, Alpha=@n31, Translation="-5,0,0", Rotation="0,0,0", Scale="1,1,1") {
    rotationMode=BMM_Additive
    rotationSpace=BCS_BoneSpace
    translationMode=BMM_Additive
    translationSpace=BCS_BoneSpace
    alphaBoolBlend {blendOption=Linear}
    alphaScaleBiasClamp {bInterpResult=true}
    boneToModify {boneName=back_coat_helper_02}
    translation {x=-5}
   }
   n34 = "Transform (Modify) Bone - Bone: back_coat_helper_03"(ComponentPose=@n33, Alpha=@n31, Translation="-3,0,0", Rotation="0,0,0", Scale="1,1,1") {
    rotationMode=BMM_Additive
    rotationSpace=BCS_BoneSpace
    translationMode=BMM_Additive
    translationSpace=BCS_BoneSpace
    alphaBoolBlend {blendOption=Linear}
    alphaScaleBiasClamp {bInterpResult=true}
    boneToModify {boneName=back_coat_helper_03}
    translation {x=-3}
   }
   n35 = "Transform (Modify) Bone - Bone: attach_backpack"(ComponentPose=@n34, Alpha=1, Translation=Get backpack_translation(), Rotation=Get backpack_rotate(), Scale="1,1,1") {
    rotationMode=BMM_Additive
    rotationSpace=BCS_BoneSpace
    translationMode=BMM_Additive
    translationSpace=BCS_BoneSpace
    alphaBoolBlend {blendOption=Linear}
    boneToModify {boneName=attach_backpack}
   }
   n36 = "Transform (Modify) Bone - Bone: floor_joint"(ComponentPose=@n35, Alpha=Get Standing Fixes Coat(), Translation="0,0,-13", Rotation="0,0,0", Scale="1,1,1") {
    translationMode=BMM_Additive
    translationSpace=BCS_BoneSpace
    alphaBoolBlend {blendOption=Linear}
    boneToModify {boneName=floor_joint}
    translation {z=-13}
   }
   n37 = Get Left() -> (Left:double)
   n38 = Get Right() -> (Right:double)
   n39 = Get Backward() -> (Backward:double)
   n40 = Get Is Crouching() -> (Is Crouching:double)
   n41 = "Transform (Modify) Bone - Bone: dyn_cape_f_r_base"(
    ComponentPose=Local To Component(LocalPose=Use cached pose 'ControlRig '())
    Alpha=@n40
    Translation="0,0,0"
    Rotation="10,0,0"
    Scale="1,1,1"
   ) {
    rotationMode=BMM_Additive
    rotationSpace=BCS_BoneSpace
    translationMode=BMM_Additive
    translationSpace=BCS_BoneSpace
    alphaBoolBlend {blendOption=Linear}
    alphaScaleBiasClamp {bInterpResult=true}
    boneToModify {boneName=dyn_cape_f_r_base}
    rotation {pitch=10}
   }
   n42 = "Transform (Modify) Bone - Bone: dyn_cape_f_r_02"(ComponentPose=@n41, Alpha=@n40, Translation="0,0,0", Rotation="30,0,0", Scale="1,1,1") {
    rotationMode=BMM_Additive
    rotationSpace=BCS_BoneSpace
    translationMode=BMM_Additive
    translationSpace=BCS_BoneSpace
    alphaBoolBlend {blendOption=Linear}
    alphaScaleBiasClamp {bInterpResult=true}
    boneToModify {boneName=dyn_cape_f_r_02}
    rotation {pitch=30}
   }
   n43 = "Transform (Modify) Bone - Bone: dyn_cape_f_r_03"(ComponentPose=@n42, Alpha=@n40, Translation="0,0,0", Rotation="20,0,0", Scale="1,1,1") {
    rotationMode=BMM_Additive
    rotationSpace=BCS_BoneSpace
    translationMode=BMM_Additive
    translationSpace=BCS_BoneSpace
    alphaBoolBlend {blendOption=Linear}
    alphaScaleBiasClamp {bInterpResult=true}
    boneToModify {boneName=dyn_cape_f_r_03}
    rotation {pitch=20}
   }
   n44 = "Transform (Modify) Bone - Bone: dyn_cape_f_r_04"(ComponentPose=@n43, Alpha=@n40, Translation="0,0,0", Rotation="20,0,0", Scale="1,1,1") {
    rotationMode=BMM_Additive
    rotationSpace=BCS_BoneSpace
    translationMode=BMM_Additive
    translationSpace=BCS_BoneSpace
    alphaBoolBlend {blendOption=Linear}
    alphaScaleBiasClamp {bInterpResult=true}
    boneToModify {boneName=dyn_cape_f_r_04}
    rotation {pitch=20}
   }
   n45 = "Transform (Modify) Bone - Bone: dyn_cape_f_r_05"(ComponentPose=@n44, Alpha=@n40, Translation="0,0,0", Rotation="0,0,0", Scale="1,1,1") {
    rotationMode=BMM_Additive
    rotationSpace=BCS_BoneSpace
    translationMode=BMM_Additive
    translationSpace=BCS_BoneSpace
    alphaBoolBlend {blendOption=Linear}
    alphaScaleBiasClamp {bInterpResult=true}
    boneToModify {boneName=dyn_cape_f_r_05}
   }
   n46 = "Transform (Modify) Bone - Bone: dyn_cape_f_l_base"(ComponentPose=@n45, Alpha=@n40, Translation="0,0,0", Rotation="0,0,0", Scale="1,1,1") {
    rotationMode=BMM_Additive
    rotationSpace=BCS_BoneSpace
    translationMode=BMM_Additive
    translationSpace=BCS_BoneSpace
    alphaBoolBlend {blendOption=Linear}
    alphaScaleBiasClamp {bInterpResult=true}
    boneToModify {boneName=dyn_cape_f_l_base}
   }
   n47 = "Transform (Modify) Bone - Bone: dyn_cape_f_l_02"(ComponentPose=@n46, Alpha=@n40, Translation="0,0,0", Rotation="30,0,0", Scale="1,1,1") {
    rotationMode=BMM_Additive
    rotationSpace=BCS_BoneSpace
    translationMode=BMM_Additive
    translationSpace=BCS_BoneSpace
    alphaBoolBlend {blendOption=Linear}
    alphaScaleBiasClamp {bInterpResult=true}
    boneToModify {boneName=dyn_cape_f_l_02}
    rotation {pitch=30}
   }
   n48 = "Transform (Modify) Bone - Bone: dyn_cape_f_l_03"(ComponentPose=@n47, Alpha=@n40, Translation="0,0,0", Rotation="10,0,0", Scale="1,1,1") {
    rotationMode=BMM_Additive
    rotationSpace=BCS_BoneSpace
    translationMode=BMM_Additive
    translationSpace=BCS_BoneSpace
    alphaBoolBlend {blendOption=Linear}
    alphaScaleBiasClamp {bInterpResult=true}
    boneToModify {boneName=dyn_cape_f_l_03}
    rotation {pitch=10}
   }
   n49 = "Transform (Modify) Bone - Bone: dyn_cape_f_l_04"(ComponentPose=@n48, Alpha=@n40, Translation="0,0,0", Rotation="20,0,0", Scale="1,1,1") {
    rotationMode=BMM_Additive
    rotationSpace=BCS_BoneSpace
    translationMode=BMM_Additive
    translationSpace=BCS_BoneSpace
    alphaBoolBlend {blendOption=Linear}
    alphaScaleBiasClamp {bInterpResult=true}
    boneToModify {boneName=dyn_cape_f_l_04}
    rotation {pitch=20}
   }
   n50 = "Transform (Modify) Bone - Bone: dyn_cape_f_l_05"(ComponentPose=@n49, Alpha=@n40, Translation="0,0,0", Rotation="20,0,0", Scale="1,1,1") {
    rotationMode=BMM_Additive
    rotationSpace=BCS_BoneSpace
    translationMode=BMM_Additive
    translationSpace=BCS_BoneSpace
    alphaBoolBlend {blendOption=Linear}
    alphaScaleBiasClamp {bInterpResult=true}
    boneToModify {boneName=dyn_cape_f_l_05}
    rotation {pitch=20}
   }
   n51 = "Transform (Modify) Bone - Bone: dyn_cape_s_l_base"(ComponentPose=@n50, Alpha=@n40, Translation="0,0,0", Rotation="0,0,0", Scale="1,1,1") {
    rotationMode=BMM_Additive
    rotationSpace=BCS_BoneSpace
    translationMode=BMM_Additive
    translationSpace=BCS_BoneSpace
    alphaBoolBlend {blendOption=Linear}
    alphaScaleBiasClamp {bInterpResult=true}
    boneToModify {boneName=dyn_cape_s_l_base}
   }
   n52 = "Transform (Modify) Bone - Bone: dyn_cape_s_l_02"(ComponentPose=@n51, Alpha=@n40, Translation="0,0,0", Rotation="35,0,0", Scale="1,1,1") {
    rotationMode=BMM_Additive
    rotationSpace=BCS_BoneSpace
    translationMode=BMM_Additive
    translationSpace=BCS_BoneSpace
    alphaBoolBlend {blendOption=Linear}
    alphaScaleBiasClamp {bInterpResult=true}
    boneToModify {boneName=dyn_cape_s_l_02}
    rotation {pitch=35}
   }
   n53 = "Transform (Modify) Bone - Bone: dyn_cape_s_l_03"(ComponentPose=@n52, Alpha=@n40, Translation="0,0,0", Rotation="10,0,0", Scale="1,1,1") {
    rotationMode=BMM_Additive
    rotationSpace=BCS_BoneSpace
    translationMode=BMM_Additive
    translationSpace=BCS_BoneSpace
    alphaBoolBlend {blendOption=Linear}
    alphaScaleBiasClamp {bInterpResult=true}
    boneToModify {boneName=dyn_cape_s_l_03}
    rotation {pitch=10}
   }
   n54 = "Transform (Modify) Bone - Bone: dyn_cape_s_l_04"(ComponentPose=@n53, Alpha=@n40, Translation="0,0,0", Rotation="10,0,0", Scale="1,1,1") {
    rotationMode=BMM_Additive
    rotationSpace=BCS_BoneSpace
    translationMode=BMM_Additive
    translationSpace=BCS_BoneSpace
    alphaBoolBlend {blendOption=Linear}
    alphaScaleBiasClamp {bInterpResult=true}
    boneToModify {boneName=dyn_cape_s_l_04}
    rotation {pitch=10}
   }
   n55 = "Transform (Modify) Bone - Bone: dyn_cape_s_l_05"(ComponentPose=@n54, Alpha=@n40, Translation="0,0,0", Rotation="10,0,0", Scale="1,1,1") {
    rotationMode=BMM_Additive
    rotationSpace=BCS_BoneSpace
    translationMode=BMM_Additive
    translationSpace=BCS_BoneSpace
    alphaBoolBlend {blendOption=Linear}
    alphaScaleBiasClamp {bInterpResult=true}
    boneToModify {boneName=dyn_cape_s_l_05}
    rotation {pitch=10}
   }
   n56 = "Transform (Modify) Bone - Bone: dyn_cape_s_r_base"(ComponentPose=@n55, Alpha=@n40, Translation="0,0,0", Rotation="20,0,0", Scale="1,1,1") {
    rotationMode=BMM_Additive
    rotationSpace=BCS_BoneSpace
    translationMode=BMM_Additive
    translationSpace=BCS_BoneSpace
    alphaBoolBlend {blendOption=Linear}
    alphaScaleBiasClamp {bInterpResult=true}
    boneToModify {boneName=dyn_cape_s_r_base}
    rotation {pitch=20}
   }
   n57 = "Transform (Modify) Bone - Bone: dyn_cape_s_r_02"(ComponentPose=@n56, Alpha=@n40, Translation="0,0,0", Rotation="20,0,0", Scale="1,1,1") {
    rotationMode=BMM_Additive
    rotationSpace=BCS_BoneSpace
    translationMode=BMM_Additive
    translationSpace=BCS_BoneSpace
    alphaBoolBlend {blendOption=Linear}
    alphaScaleBiasClamp {bInterpResult=true}
    boneToModify {boneName=dyn_cape_s_r_02}
    rotation {pitch=20}
   }
   n58 = "Transform (Modify) Bone - Bone: dyn_cape_s_r_03"(ComponentPose=@n57, Alpha=@n40, Translation="0,0,0", Rotation="20,0,0", Scale="1,1,1") {
    rotationMode=BMM_Additive
    rotationSpace=BCS_BoneSpace
    translationMode=BMM_Additive
    translationSpace=BCS_BoneSpace
    alphaBoolBlend {blendOption=Linear}
    alphaScaleBiasClamp {bInterpResult=true}
    boneToModify {boneName=dyn_cape_s_r_03}
    rotation {pitch=20}
   }
   n59 = "Transform (Modify) Bone - Bone: dyn_cape_s_r_04"(ComponentPose=@n58, Alpha=@n40, Translation="0,0,0", Rotation="20,0,0", Scale="1,1,1") {
    rotationMode=BMM_Additive
    rotationSpace=BCS_BoneSpace
    translationMode=BMM_Additive
    translationSpace=BCS_BoneSpace
    alphaBoolBlend {blendOption=Linear}
    alphaScaleBiasClamp {bInterpResult=true}
    boneToModify {boneName=dyn_cape_s_r_04}
    rotation {pitch=20}
   }
   n60 = "Transform (Modify) Bone - Bone: dyn_cape_s_r_05"(ComponentPose=@n59, Alpha=@n40, Translation="0,0,0", Rotation="0,0,0", Scale="1,1,1") {
    rotationMode=BMM_Additive
    rotationSpace=BCS_BoneSpace
    translationMode=BMM_Additive
    translationSpace=BCS_BoneSpace
    alphaBoolBlend {blendOption=Linear}
    alphaScaleBiasClamp {bInterpResult=true}
    boneToModify {boneName=dyn_cape_s_r_05}
   }
   n61 = "Transform (Modify) Bone - Bone: dyn_cape_b_l_base"(ComponentPose=@n60, Alpha=@n40, Translation="0,0,0", Rotation="0,0,0", Scale="1,1,1") {
    rotationMode=BMM_Additive
    rotationSpace=BCS_BoneSpace
    translationMode=BMM_Additive
    translationSpace=BCS_BoneSpace
    alphaBoolBlend {blendOption=Linear}
    alphaScaleBiasClamp {bInterpResult=true}
    boneToModify {boneName=dyn_cape_b_l_base}
   }
   n62 = "Transform (Modify) Bone - Bone: dyn_cape_b_l_02"(ComponentPose=@n61, Alpha=@n40, Translation="0,0,0", Rotation="25,0,0", Scale="1,1,1") {
    rotationMode=BMM_Additive
    rotationSpace=BCS_BoneSpace
    translationMode=BMM_Additive
    translationSpace=BCS_BoneSpace
    alphaBoolBlend {blendOption=Linear}
    alphaScaleBiasClamp {bInterpResult=true}
    boneToModify {boneName=dyn_cape_b_l_02}
    rotation {pitch=25}
   }
   n63 = "Transform (Modify) Bone - Bone: dyn_cape_b_l_03"(ComponentPose=@n62, Alpha=@n40, Translation="0,0,0", Rotation="20,0,0", Scale="1,1,1") {
    rotationMode=BMM_Additive
    rotationSpace=BCS_BoneSpace
    translationMode=BMM_Additive
    translationSpace=BCS_BoneSpace
    alphaBoolBlend {blendOption=Linear}
    alphaScaleBiasClamp {bInterpResult=true}
    boneToModify {boneName=dyn_cape_b_l_03}
    rotation {pitch=20}
   }
   n64 = "Transform (Modify) Bone - Bone: dyn_cape_b_l_04"(ComponentPose=@n63, Alpha=@n40, Translation="0,0,0", Rotation="20,0,0", Scale="1,1,1") {
    rotationMode=BMM_Additive
    rotationSpace=BCS_BoneSpace
    translationMode=BMM_Additive
    translationSpace=BCS_BoneSpace
    alphaBoolBlend {blendOption=Linear}
    alphaScaleBiasClamp {bInterpResult=true}
    boneToModify {boneName=dyn_cape_b_l_04}
    rotation {pitch=20}
   }
   n65 = "Transform (Modify) Bone - Bone: dyn_cape_b_l_05"(ComponentPose=@n64, Alpha=@n40, Translation="0,0,0", Rotation="20,0,0", Scale="1,1,1") {
    rotationMode=BMM_Additive
    rotationSpace=BCS_BoneSpace
    translationMode=BMM_Additive
    translationSpace=BCS_BoneSpace
    alphaBoolBlend {blendOption=Linear}
    alphaScaleBiasClamp {bInterpResult=true}
    boneToModify {boneName=dyn_cape_b_l_05}
    rotation {pitch=20}
   }
   n66 = "Transform (Modify) Bone - Bone: dyn_cape_b_r_base"(ComponentPose=@n65, Alpha=@n40, Translation="0,0,0", Rotation="0,0,0", Scale="1,1,1") {
    rotationMode=BMM_Additive
    rotationSpace=BCS_BoneSpace
    translationMode=BMM_Additive
    translationSpace=BCS_BoneSpace
    alphaBoolBlend {blendOption=Linear}
    alphaScaleBiasClamp {bInterpResult=true}
    boneToModify {boneName=dyn_cape_b_r_base}
   }
   n67 = "Transform (Modify) Bone - Bone: dyn_cape_b_r_02"(ComponentPose=@n66, Alpha=@n40, Translation="0,0,0", Rotation="30,0,0", Scale="1,1,1") {
    rotationMode=BMM_Additive
    rotationSpace=BCS_BoneSpace
    translationMode=BMM_Additive
    translationSpace=BCS_BoneSpace
    alphaBoolBlend {blendOption=Linear}
    alphaScaleBiasClamp {bInterpResult=true}
    boneToModify {boneName=dyn_cape_b_r_02}
    rotation {pitch=30}
   }
   n68 = "Transform (Modify) Bone - Bone: dyn_cape_b_r_03"(ComponentPose=@n67, Alpha=@n40, Translation="0,0,0", Rotation="20,0,0", Scale="1,1,1") {
    rotationMode=BMM_Additive
    rotationSpace=BCS_BoneSpace
    translationMode=BMM_Additive
    translationSpace=BCS_BoneSpace
    alphaBoolBlend {blendOption=Linear}
    alphaScaleBiasClamp {bInterpResult=true}
    boneToModify {boneName=dyn_cape_b_r_03}
    rotation {pitch=20}
   }
   n69 = "Transform (Modify) Bone - Bone: dyn_cape_b_r_04"(ComponentPose=@n68, Alpha=@n40, Translation="0,0,0", Rotation="20,0,0", Scale="1,1,1") {
    rotationMode=BMM_Additive
    rotationSpace=BCS_BoneSpace
    translationMode=BMM_Additive
    translationSpace=BCS_BoneSpace
    alphaBoolBlend {blendOption=Linear}
    alphaScaleBiasClamp {bInterpResult=true}
    boneToModify {boneName=dyn_cape_b_r_04}
    rotation {pitch=20}
   }
   n70 = "Transform (Modify) Bone - Bone: dyn_cape_b_r_05"(ComponentPose=@n69, Alpha=@n40, Translation="0,0,0", Rotation="20,0,0", Scale="1,1,1") {
    rotationMode=BMM_Additive
    rotationSpace=BCS_BoneSpace
    translationMode=BMM_Additive
    translationSpace=BCS_BoneSpace
    alphaBoolBlend {blendOption=Linear}
    alphaScaleBiasClamp {bInterpResult=true}
    boneToModify {boneName=dyn_cape_b_r_05}
    rotation {pitch=20}
   }
   n71 = Get Forward() -> (Forward:double)
   n72 = "Transform (Modify) Bone - Bone: dyn_cape_b_l_base"(ComponentPose=@n70, Alpha=@n71, Translation="0,0,0", Rotation="33,0,0", Scale="1,1,1") {
    rotationMode=BMM_Additive
    rotationSpace=BCS_BoneSpace
    translationMode=BMM_Additive
    translationSpace=BCS_BoneSpace
    alphaBoolBlend {blendOption=Linear}
    alphaScaleBiasClamp {bInterpResult=true}
    boneToModify {boneName=dyn_cape_b_l_base}
    rotation {pitch=33}
   }
   n73 = "Transform (Modify) Bone - Bone: dyn_cape_b_r_base"(ComponentPose=@n72, Alpha=@n71, Translation="0,0,0", Rotation="33,0,0", Scale="1,1,1") {
    rotationMode=BMM_Additive
    rotationSpace=BCS_BoneSpace
    translationMode=BMM_Additive
    translationSpace=BCS_BoneSpace
    alphaBoolBlend {blendOption=Linear}
    alphaScaleBiasClamp {bInterpResult=true}
    boneToModify {boneName=dyn_cape_b_r_base}
    rotation {pitch=33}
   }
   n74 = "Transform (Modify) Bone - Bone: dyn_cape_s_l_base"(ComponentPose=@n73, Alpha=@n71, Translation="0,0,0", Rotation="10,0,0", Scale="1,1,1") {
    rotationMode=BMM_Additive
    rotationSpace=BCS_BoneSpace
    translationMode=BMM_Additive
    translationSpace=BCS_BoneSpace
    alphaBoolBlend {blendOption=Linear}
    alphaScaleBiasClamp {bInterpResult=true}
    boneToModify {boneName=dyn_cape_s_l_base}
    rotation {pitch=10}
   }
   n75 = "Transform (Modify) Bone - Bone: dyn_cape_s_r_base"(ComponentPose=@n74, Alpha=@n71, Translation="0,0,0", Rotation="10,0,0", Scale="1,1,1") {
    rotationMode=BMM_Additive
    rotationSpace=BCS_BoneSpace
    translationMode=BMM_Additive
    translationSpace=BCS_BoneSpace
    alphaBoolBlend {blendOption=Linear}
    alphaScaleBiasClamp {bInterpResult=true}
    boneToModify {boneName=dyn_cape_s_r_base}
    rotation {pitch=10}
   }
   n76 = "Transform (Modify) Bone - Bone: back_coat_helper_02"(ComponentPose=@n75, Alpha=@n71, Translation="-15,0,0", Rotation="0,0,0", Scale="1,1,1") {
    rotationMode=BMM_Additive
    rotationSpace=BCS_BoneSpace
    translationMode=BMM_Additive
    translationSpace=BCS_BoneSpace
    alphaBoolBlend {blendOption=Linear}
    alphaScaleBiasClamp {bInterpResult=true}
    boneToModify {boneName=back_coat_helper_02}
    translation {x=-15}
   }
   n77 = "Transform (Modify) Bone - Bone: dyn_cape_b_l_base"(ComponentPose=@n76, Alpha=@n39, Translation="0,0,0", Rotation="25,0,0", Scale="1,1,1") {
    rotationMode=BMM_Additive
    rotationSpace=BCS_BoneSpace
    translationMode=BMM_Additive
    translationSpace=BCS_BoneSpace
    alphaBoolBlend {blendOption=Linear}
    alphaScaleBiasClamp {bInterpResult=true}
    boneToModify {boneName=dyn_cape_b_l_base}
    rotation {pitch=25}
   }
   n78 = "Transform (Modify) Bone - Bone: dyn_cape_b_r_base"(ComponentPose=@n77, Alpha=@n39, Translation="0,0,0", Rotation="25,0,0", Scale="1,1,1") {
    rotationMode=BMM_Additive
    rotationSpace=BCS_BoneSpace
    translationMode=BMM_Additive
    translationSpace=BCS_BoneSpace
    alphaBoolBlend {blendOption=Linear}
    alphaScaleBiasClamp {bInterpResult=true}
    boneToModify {boneName=dyn_cape_b_r_base}
    rotation {pitch=25}
   }
   n79 = "Transform (Modify) Bone - Bone: dyn_cape_s_l_base"(ComponentPose=@n78, Alpha=@n39, Translation="0,0,0", Rotation="10,0,0", Scale="1,1,1") {
    rotationMode=BMM_Additive
    rotationSpace=BCS_BoneSpace
    translationMode=BMM_Additive
    translationSpace=BCS_BoneSpace
    alphaBoolBlend {blendOption=Linear}
    alphaScaleBiasClamp {bInterpResult=true}
    boneToModify {boneName=dyn_cape_s_l_base}
    rotation {pitch=10}
   }
   n80 = "Transform (Modify) Bone - Bone: dyn_cape_s_r_base"(ComponentPose=@n79, Alpha=@n39, Translation="0,0,0", Rotation="10,0,0", Scale="1,1,1") {
    rotationMode=BMM_Additive
    rotationSpace=BCS_BoneSpace
    translationMode=BMM_Additive
    translationSpace=BCS_BoneSpace
    alphaBoolBlend {blendOption=Linear}
    alphaScaleBiasClamp {bInterpResult=true}
    boneToModify {boneName=dyn_cape_s_r_base}
    rotation {pitch=10}
   }
   n81 = "Transform (Modify) Bone - Bone: dyn_cape_f_r_base"(ComponentPose=@n80, Alpha=@n38, Translation="0,0,0", Rotation="15,0,0", Scale="1,1,1") {
    rotationMode=BMM_Additive
    rotationSpace=BCS_BoneSpace
    translationMode=BMM_Additive
    translationSpace=BCS_BoneSpace
    alphaBoolBlend {blendOption=Linear}
    alphaScaleBiasClamp {bInterpResult=true}
    boneToModify {boneName=dyn_cape_f_r_base}
    rotation {pitch=15}
   }
   n82 = "Transform (Modify) Bone - Bone: dyn_cape_s_r_base"(ComponentPose=@n81, Alpha=@n38, Translation="0,0,0", Rotation="15,0,0", Scale="1,1,1") {
    rotationMode=BMM_Additive
    rotationSpace=BCS_BoneSpace
    translationMode=BMM_Additive
    translationSpace=BCS_BoneSpace
    alphaBoolBlend {blendOption=Linear}
    alphaScaleBiasClamp {bInterpResult=true}
    boneToModify {boneName=dyn_cape_s_r_base}
    rotation {pitch=15}
   }
   n83 = "Transform (Modify) Bone - Bone: dyn_cape_f_l_base"(ComponentPose=@n82, Alpha=@n38, Translation="0,0,0", Rotation="8,0,0", Scale="1,1,1") {
    rotationMode=BMM_Additive
    rotationSpace=BCS_BoneSpace
    translationMode=BMM_Additive
    translationSpace=BCS_BoneSpace
    alphaBoolBlend {blendOption=Linear}
    alphaScaleBiasClamp {bInterpResult=true}
    boneToModify {boneName=dyn_cape_f_l_base}
    rotation {pitch=8}
   }
   n84 = "Transform (Modify) Bone - Bone: dyn_cape_s_l_base"(ComponentPose=@n83, Alpha=@n38, Translation="0,0,0", Rotation="8,0,0", Scale="1,1,1") {
    rotationMode=BMM_Additive
    rotationSpace=BCS_BoneSpace
    translationMode=BMM_Additive
    translationSpace=BCS_BoneSpace
    alphaBoolBlend {blendOption=Linear}
    alphaScaleBiasClamp {bInterpResult=true}
    boneToModify {boneName=dyn_cape_s_l_base}
    rotation {pitch=8}
   }
   n85 = "Transform (Modify) Bone - Bone: dyn_cape_f_r_base"(ComponentPose=@n84, Alpha=@n37, Translation="0,0,0", Rotation="15,0,0", Scale="1,1,1") {
    rotationMode=BMM_Additive
    rotationSpace=BCS_BoneSpace
    translationMode=BMM_Additive
    translationSpace=BCS_BoneSpace
    alphaBoolBlend {blendOption=Linear}
    alphaScaleBiasClamp {bInterpResult=true}
    boneToModify {boneName=dyn_cape_f_r_base}
    rotation {pitch=15}
   }
   n86 = "Transform (Modify) Bone - Bone: dyn_cape_s_r_base"(ComponentPose=@n85, Alpha=@n37, Translation="0,0,0", Rotation="15,0,0", Scale="1,1,1") {
    rotationMode=BMM_Additive
    rotationSpace=BCS_BoneSpace
    translationMode=BMM_Additive
    translationSpace=BCS_BoneSpace
    alphaBoolBlend {blendOption=Linear}
    alphaScaleBiasClamp {bInterpResult=true}
    boneToModify {boneName=dyn_cape_s_r_base}
    rotation {pitch=15}
   }
   n87 = "Transform (Modify) Bone - Bone: dyn_cape_f_l_base"(ComponentPose=@n86, Alpha=@n37, Translation="0,0,0", Rotation="8,0,0", Scale="1,1,1") {
    rotationMode=BMM_Additive
    rotationSpace=BCS_BoneSpace
    translationMode=BMM_Additive
    translationSpace=BCS_BoneSpace
    alphaBoolBlend {blendOption=Linear}
    alphaScaleBiasClamp {bInterpResult=true}
    boneToModify {boneName=dyn_cape_f_l_base}
    rotation {pitch=8}
   }
   n88 = "Transform (Modify) Bone - Bone: dyn_cape_s_l_base"(ComponentPose=@n87, Alpha=@n37, Translation="0,0,0", Rotation="8,0,0", Scale="1,1,1") {
    rotationMode=BMM_Additive
    rotationSpace=BCS_BoneSpace
    translationMode=BMM_Additive
    translationSpace=BCS_BoneSpace
    alphaBoolBlend {blendOption=Linear}
    alphaScaleBiasClamp {bInterpResult=true}
    boneToModify {boneName=dyn_cape_s_l_base}
    rotation {pitch=8}
   }
   Save cached pose 'Pre-dynamics'(Pose=Blend Poses by bool(BlendPose_0=Component To Local(ComponentPose=@n36), BlendPose_1=Component To Local(ComponentPose=@n88), bActiveValue=Get bRigidBodyEnabled()) {blendType=Linear}) {cacheName=Pre-dynamics}
   Reroute Node(InputPin:Wildcard) -> (OutputPin:Wildcard)
  ]}
 EventGraph {nodes [
   n1 = Event BlueprintUpdateAnimation()
   > n2 = Sequence!(execute) -> (then_0, then_1, then_2:Exec, then_3:Exec, then_4:Exec, then_5:Exec, then_6:Exec, then_7:Exec, then_8:Exec)
   > n3 = Gravity Override State Function DEPRECATED(CharacterPartAnimInstance=Self-Reference(), GameTimeInSeconds=GetGameTimeInSeconds(), GravityOverrideParams=Get Gravity Override Params())
   > Set Gravity Override(Gravity Override=@n3)
   n4 = Get bIsCrouching()
   n5 = Get bIsMoving2D()
   > Set Is Crouching(Is Crouching:double=SelectFloat(A=0.5, B="To Float (Boolean)"(InBool=@n4), bPickA=AND Boolean(A=@n4, B=@n5))) -> (Output_Get:double)
   > Set Is Sliding(Is Sliding:double="To Float (Boolean)"(InBool=Get bIsSliding())) -> (Output_Get:double)
   > Set is SkyDiving Dive Mode(is SkyDiving Dive Mode:double="To Float (Boolean)"(InBool=Get bIsSkydiveDiveMode())) -> (Output_Get:double)
   > Set Is Parachuting(Is Parachuting:double="To Float (Boolean)"(InBool=Get bIsParachuteOpened())) -> (Output_Get:double)
   > Set Is Skydiving(Is Skydiving:double="To Float (Boolean)"(InBool=Get bIsSkydiveFloating())) -> (Output_Get:double)
   n6 = Get bIsSwimSprinting() -> (bIsSwimSprinting:bool)
   n7 = Get bIsSkyDiving()
   n8 = Get bIsCrouching()
   > n9 = Branch(Condition=AND Boolean!(A=Get bIsMoving2D(), B=NOT Boolean(A=@n7), C:bool=NOT Boolean(A=@n8), D:bool=NOT Boolean(A=@n6)) -> (ReturnValue))
   > Set Moving2D BackHelper(Moving2D BackHelper:double=1) -> (Output_Get:double)
   Set Moving2D BackHelper(execute=@n9.else, Moving2D BackHelper=0)
   > Set is FrontEnd(is FrontEnd:double="To Float (Boolean)"(InBool=Get bIsInFrontEnd())) -> (Output_Get:double)
   n10 = Get bIsInVehicle() -> (bIsInVehicle:bool)
   n11 = Get bIsInVehicleSeat() -> (bIsInVehicleSeat:bool)
   Set Inside Car(execute=@n2.then_8, Inside Car:double="To Float (Boolean)"(InBool=AND Boolean!(A=@n10, B=@n11) -> (ReturnValue))) -> (Output_Get:double)
   n12 = OR Boolean(A=Get bIsJumping(), B=Get bIsFalling())
   n13 = Get bIsTacticalSprinting()
   n14 = Branch(execute=@n2.then_1, Condition=OR Boolean(A=Get bIsCrouching(), B=AND Boolean(A=@n13, B=@n12)))
   > Set Standing Fixes Coat(Standing Fixes Coat:double=0) -> (Output_Get:double)
   Set Standing Fixes Coat(execute=@n14.else, Standing Fixes Coat=1)
   n15 = Branch(execute=@n2.then_4, Condition=Get bIsCrouching())
   n16 = Get Moving 2d()
   > Set Moving 2d(Moving 2d:double=@n16) -> (Output_Get:double)
   n17 = AND Boolean!(
    A=Get bIsMoving2D()
    B=NOT Boolean(A=Get bIsSkydiveDiveMode())
    C:bool=NOT Boolean(A=Get bIsSkydiveFloating())
    D:bool=NOT Boolean(A=Get bIsSkyDiving())
    E:bool=NOT Boolean(A=Get bIsDBNO())
    F:bool=NOT Boolean(A=Get bIsSwimSprinting())
   ) -> (ReturnValue)
   Set Moving 2d(execute=@n15.else, Moving 2d=FInterpTo(Current=@n16, Target="To Float (Boolean)"(InBool=@n17), DeltaTime=@n1.DeltaTimeX, InterpSpeed=20))
   n18 = PlayerMovement(execute=@n2.then_2, AnimInstance=Self-Reference(), XDotProductCoefficient=100, YDotProductCoefficient=100, ZDotProductCoefficient=100)
   > n19 = Branch(Condition=Get bIsCrouching())
   n20 = BreakVector(InVec=MakeVector(X=@n18, Y=@n18.YMovement))
   Set Left(execute=@n19.else, Left:double="To Float (Boolean)"(InBool="float < float"(A=@n20, B=-50))) -> (Output_Get:double)
   > Set Right(Right:double="To Float (Boolean)"(InBool="float > float"(A=@n20, B=50))) -> (Output_Get:double)
   > Set Forward(Forward:double="To Float (Boolean)"(InBool="float > float"(A=@n20.Y, B=40))) -> (Output_Get:double)
   > Set Backward(Backward:double="To Float (Boolean)"(InBool="float < float"(A=@n20.Y, B=-40))) -> (Output_Get:double)
   n21 = GetPartSkeletalMeshComponent()
   n22 = "Not Equal (Object)"(
    A:Object Reference=FindChildCommponentFromAttachSocket!(
     self:Custom Character Part Anim Instance Object Reference
     Parent:Skeletal Mesh Component Object Reference=@n21
     bIncludeAllDescendants:bool=false
     AttachSocket:Name=backpack_br
    ) -> (ReturnValue:Skeletal Mesh Component Object Reference)
    B:Object Reference
   ) -> (ReturnValue:bool=~) {is=title K2Node_PromotableOperator}
   n23 = FindChildCommponentFromAttachSocket(Parent=@n21, bIncludeAllDescendants=false, AttachSocket=Cape)
   Set Is Backbling Equiped(
    execute=[@n2.then_3, @n2.then_6]
    Is Backbling Equiped:bool=OR Boolean(A="Not Equal (Object)"(A=@n23) -> (ReturnValue=~), B=NOT Boolean(A=@n22))
   ) -> (Output_Get:bool)
   > Set Is Skydive Cape Equiped(Is Skydive Cape Equiped:bool=AND Boolean!(A=Get Is Backbling Equiped() -> (Is Backbling Equiped:bool), B=Get bIsSkyDiving()) -> (ReturnValue)) -> (Output_Get:bool)
   n24 = Get bIsMoving2D()
   Set Moving 2D Back(
    execute=@n2.then_5
    Moving 2D Back:double=FInterpTo(Current=Get Moving 2D Back() -> (Moving 2D Back:double), Target="To Float (Boolean)"(InBool=@n24), DeltaTime=@n1.DeltaTimeX, InterpSpeed=0.6)
   ) -> (Output_Get:double)
   n25 = GetOwnerSkeletalMeshComponent()
   n26 = BreakRotator(InRot=BreakTransform(InTransform=GetDeltaTransformFromRefPose(self=@n25, BoneName=spine_05, BaseName=spine_01)).Rotation)
   Set backpack_rotate(execute=@n2.then_7, backpack_rotate=MakeRotator(Roll=MapRangeClamped(Value=@n26.Yaw, InRangeA=0, InRangeB=-30, OutRangeB=35)))
   > Set backpack_translation(backpack_translation=MakeVector(Y=MapRangeClamped(Value=@n26.Yaw, InRangeA=0, InRangeB=-30, OutRangeB=-7)))
   Event BlueprintInitializeAnimation()
   > Set Skel Mesh(Skel Mesh=GetOwnerSkeletalMeshComponent())
   TryGetPawnOwner()
  ]}
]
variables [
 backpack_rotate {type=Rotator}
 backpack_translation {type=Vector}
 Backward {type=number}
 CrouchingMoving {type=number}
 CrouchMoving2d {type=number}
 Forward {type=number}
 Gravity Override {type=Vector}
 Gravity Override Params {type=GravityOverrideParamsStruct, default {front_End {gravityOverride=<0 0 -2500>}}}
 Inside Car {type=number}
 Is Backbling Equiped {type=boolean}
 Is Crouching {type=number}
 is FrontEnd {type=number}
 Is Parachuting {type=number}
 Is Skydive Cape Equiped {type=boolean}
 Is Skydiving {type=number}
 is SkyDiving Dive Mode {type=number}
 Is Sliding {type=number}
 Left {type=number}
 Moving 2d {type=number}
 Moving 2D Back {type=number}
 Moving 2d Ground {type=number}
 Moving 2dBool {type=boolean}
 Moving Forward Standing {type=number}
 Moving2D BackHelper {type=number}
 NOT Skydive {type=boolean}
 Right {type=number}
 Sim Space Settings {
  type=SimSpaceSettings
  default {
   dampingAlpha=1
   maxAngularAcceleration=1000
   maxAngularVelocity=3
   maxLinearAcceleration=700
   maxLinearVelocity=500
   velocityScaleZ=0
   worldAlpha=0
   externalLinearDragV=<0.2 0.2 0>
  }
 }
 Sim Space Settings_Crouching {
  type=SimSpaceSettings
  default {
   dampingAlpha=1
   maxAngularAcceleration=1000
   maxAngularVelocity=3
   maxLinearAcceleration=700
   maxLinearVelocity=500
   velocityScaleZ=0
   worldAlpha=0
   externalLinearDragV=<0.2 0.2 0>
  }
 }
 Sim Space Settings_Regular {
  type=SimSpaceSettings
  default {
   dampingAlpha=1
   maxAngularAcceleration=1000
   maxAngularVelocity=3
   maxLinearAcceleration=700
   maxLinearVelocity=500
   velocityScaleZ=0
   worldAlpha=0
   externalLinearDragV=<0.2 0.2 0>
  }
 }
 Skel Mesh {type=SkeletalMeshComponent}
 Standing Fixes Coat {type=number}
]
