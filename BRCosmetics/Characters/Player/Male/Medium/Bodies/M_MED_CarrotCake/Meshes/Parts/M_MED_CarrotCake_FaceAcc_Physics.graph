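# Physics asset for the CarrotCake male medium character's face accessory mesh, giving its back, left and right hair strands jointed physics bodies.
# Each strand chains a small base capsule on the head to two dynamic capsules, with the head joined to the neck.
# Base and head joints swing and twist freely, while dynamic links have tight swing and twist limits.
# The left tip adds a second, angled capsule, and large capsules cover head and neck.
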PhysicsAsset /BRCosmetics/Characters/Player/Male/Medium/Bodies/M_MED_CarrotCake/Meshes/Parts/M_MED_CarrotCake_FaceAcc_Physics
bodies {
 base_hair_b_01 [{shapeType=Capsule, length=0.505, radius=0.505, shapeName=base_hair_b_01_capsule, rotation=<rot 0 -90 0>}]
 base_hair_l_01 [{shapeType=Capsule, length=0.505, radius=0.505, shapeName=base_hair_l_01_capsule, rotation=<rot 0 -90 0>}]
 base_hair_r_01 [{shapeType=Capsule, length=0.505, radius=0.505, shapeName=base_hair_r_01_capsule, rotation=<rot 0 -90 0>}]
 dyn_hair_b_01 [{shapeType=Capsule, length=2, radius=1, shapeName=dyn_hair_b_01_capsule, center=<2 0 0>, rotation=<rot 90 0 0>}]
 dyn_hair_b_02 [{shapeType=Capsule, length=4.0166, radius=1, shapeName=dyn_hair_b_02_capsule, center=<2.8839 0 0>, rotation=<rot 90 0 0>}]
 dyn_hair_l_01 [{shapeType=Capsule, length=2, radius=1, shapeName=dyn_hair_l_01_capsule, center=<2 0 0>, rotation=<rot 90 0 0>}]
 dyn_hair_l_02 [
  {shapeType=Capsule, length=3.3553, radius=1, shapeName=dyn_hair_l_02_capsule, center=<2.5866 0 0>, rotation=<rot 90 0 0>}
  {
   shapeType=Capsule
   length=2
   radius=1.6391
   shapeName=dyn_hair_l_02_capsule
   center=<3.0645 1.3285 3.1038>
   rotation=<rot 9.8471 9.1825 43.3874>
  }
 ]
 dyn_hair_r_01 [{shapeType=Capsule, length=2, radius=1, shapeName=dyn_hair_r_01_capsule, center=<2 0 0>, rotation=<rot 90 0 0>}]
 dyn_hair_r_02 [{shapeType=Capsule, length=3.4903, radius=1, shapeName=dyn_hair_r_02_capsule, center=<2.2674 0 0>, rotation=<rot 90 0 0>}]
 head [{shapeType=Capsule, length=1.2376, radius=7.2715, shapeName=head_capsule, center=<1.4543 3.377 0>, rotation=<rot 90 0 0>}]
 neck_02 [{shapeType=Capsule, length=2.0112, radius=5.1301, shapeName=neck_02_capsule, center=<2.7381 1.0427 0>, rotation=<rot 0 12.3373 0>}]
}
constraints [
 {
  bone1Name=head
  bone2Name=neck_02
  swing1LimitDegrees=45
  swing1Motion=Free
  swing2LimitDegrees=45
  swing2Motion=Free
  twistLimitDegrees=45
  twistMotion=Free
 }
 {
  bone1Name=base_hair_b_01
  bone2Name=head
  swing1LimitDegrees=45
  swing1Motion=Free
  swing2LimitDegrees=45
  swing2Motion=Free
  twistLimitDegrees=45
  twistMotion=Free
 }
 {bone1Name=dyn_hair_b_01, bone2Name=base_hair_b_01, swing1LimitDegrees=10, swing2LimitDegrees=5, twistLimitDegrees=1}
 {bone1Name=dyn_hair_b_02, bone2Name=dyn_hair_b_01, swing1LimitDegrees=10, swing2LimitDegrees=5, twistLimitDegrees=1}
 {
  bone1Name=base_hair_l_01
  bone2Name=head
  swing1LimitDegrees=45
  swing1Motion=Free
  swing2LimitDegrees=45
  swing2Motion=Free
  twistLimitDegrees=45
  twistMotion=Free
 }
 {bone1Name=dyn_hair_l_01, bone2Name=base_hair_l_01, swing1LimitDegrees=10, swing2LimitDegrees=3, twistLimitDegrees=1}
 {bone1Name=dyn_hair_l_02, bone2Name=dyn_hair_l_01, swing1LimitDegrees=10, swing2LimitDegrees=5, twistLimitDegrees=1}
 {
  bone1Name=base_hair_r_01
  bone2Name=head
  swing1LimitDegrees=45
  swing1Motion=Free
  swing2LimitDegrees=45
  swing2Motion=Free
  twistLimitDegrees=45
  twistMotion=Free
 }
 {bone1Name=dyn_hair_r_01, bone2Name=base_hair_r_01, swing1LimitDegrees=10, swing2LimitDegrees=3, twistLimitDegrees=1}
 {bone1Name=dyn_hair_r_02, bone2Name=dyn_hair_r_01, swing1LimitDegrees=10, swing2LimitDegrees=3, twistLimitDegrees=1}
]
